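# Physics asset for the female Myth pickaxe, a two-handed melee weapon with four dangling ribbons.
# A long capsule covers the handle, and each ribbon is a chain of small capsules constrained bone to bone back to the handle.
# The two four-segment ribbons are tightest at the handle and loosen toward the tips, mostly in swing.
# The two shorter two-segment ribbons use uniform limits along their length.
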PhysicsAsset /BRCosmetics/Weapons/FORT_Melee/Pickaxe_Myth_Female/Meshes/Myth_Female_Physics
bodies {
 Melee_TwoHanded_Handle [{shapeType=Capsule, length=47.6076, radius=2.5878, center=<0.5399 -1.3127 50.7221>, rotation=<rot 0.0118 -89.9999 -0.2351>}]
 ribbon1_01 [{shapeType=Capsule, length=1.75, radius=1, center=<1.2867 0 -0.019>, rotation=<rot -89.1521 -179.9998 179.9998>}]
 ribbon1_02 [{shapeType=Capsule, length=1.75, radius=0.75, center=<1.5779 0.078 0.3413>, rotation=<rot -89.1021 -179.9999 179.9999>}]
 ribbon1_03 [{shapeType=Capsule, length=1.5, radius=1, center=<1.4993 0 -0.022>, rotation=<rot -89.156 -179.9998 179.9999>}]
 ribbon1_04 [{shapeType=Capsule, length=1.5, radius=0.75, center=<0.0972 0 -0.0014>, rotation=<rot -89.156 -179.9998 179.9999>}]
 ribbon2_01 [{shapeType=Capsule, length=1.5, radius=0.75, center=<0.0062 0 0.4073>, rotation=<rot -89.116 -180 -179.9999>}]
 ribbon2_02 [{shapeType=Capsule, length=1.75, radius=0.75, rotation=<rot -89.13 -179.9996 -180>}]
 ribbon2_03 [{shapeType=Capsule, length=1.5, radius=1, center=<-0.3572 0.306 0.0052>, rotation=<rot -89.151 179.9999 -179.9999>}]
 ribbon2_04 [{shapeType=Capsule, length=1, radius=0.75, rotation=<rot -89.1512 179.9999 -179.9999>}]
 ribbon3_01 [{shapeType=Capsule, length=1, radius=1.75, center=<1.6726 0 0.0001>, rotation=<rot 90 0 0>}]
 ribbon3_02 [{shapeType=Capsule, length=1.25, radius=2.5, center=<2.072 0 0.0001>, rotation=<rot 90 0 0>}]
 ribbon4_01 [{shapeType=Capsule, length=1.5, radius=1.5, center=<1.7394 0 0>, rotation=<rot 90 0 0>}]
 ribbon4_02 [{shapeType=Capsule, length=1, radius=2.25, center=<2.2554 0 0>, rotation=<rot 90 0 0>}]
}
constraints [
 {bone1Name=ribbon3_01, bone2Name=Melee_TwoHanded_Handle, swing1LimitDegrees=20, swing2LimitDegrees=20, twistLimitDegrees=10}
 {bone1Name=ribbon3_02, bone2Name=ribbon3_01, swing1LimitDegrees=20, swing2LimitDegrees=20, twistLimitDegrees=10}
 {bone1Name=ribbon4_01, bone2Name=Melee_TwoHanded_Handle, swing1LimitDegrees=20, swing2LimitDegrees=20, twistLimitDegrees=10}
 {bone1Name=ribbon4_02, bone2Name=ribbon4_01, swing1LimitDegrees=20, swing2LimitDegrees=20, twistLimitDegrees=10}
 {bone1Name=ribbon1_01, bone2Name=Melee_TwoHanded_Handle, swing1LimitDegrees=3, swing2LimitDegrees=6, twistLimitDegrees=3}
 {bone1Name=ribbon1_02, bone2Name=ribbon1_01, swing1LimitDegrees=6, swing2LimitDegrees=18, twistLimitDegrees=6}
 {bone1Name=ribbon1_03, bone2Name=ribbon1_02, swing1LimitDegrees=9, swing2LimitDegrees=35, twistLimitDegrees=9}
 {bone1Name=ribbon1_04, bone2Name=ribbon1_03, swing1LimitDegrees=9, swing2LimitDegrees=35, twistLimitDegrees=9}
 {bone1Name=ribbon2_01, bone2Name=Melee_TwoHanded_Handle, swing1LimitDegrees=3, swing2LimitDegrees=6, twistLimitDegrees=6}
 {bone1Name=ribbon2_02, bone2Name=ribbon2_01, swing1LimitDegrees=3, swing2LimitDegrees=6, twistLimitDegrees=6}
 {bone1Name=ribbon2_03, bone2Name=ribbon2_02, swing1LimitDegrees=5, swing2LimitDegrees=15, twistLimitDegrees=9}
 {bone1Name=ribbon2_04, bone2Name=ribbon2_03, swing1LimitDegrees=5, swing2LimitDegrees=18, twistLimitDegrees=9}
]
