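# Animation Blueprint for the Myth_Female pickaxe mesh, adding physics-driven secondary motion to the incoming pose.
# A RigidBody node simulates the Myth_Female physics asset in base-bone space relative to b_AR_Root, with overridden gravity and scaled, clamped component motion.
# Chained additive Transform Bone nodes then offset the rotations of the four ribbon2 bones by fixed amounts after the simulation.
# The event graph has no connected logic.
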AnimBlueprint /BRCosmetics/Weapons/FORT_Melee/Pickaxe_Myth_Female/Meshes/Myth_Female_AnimBP
target_skeleton=&/Myth_Female_Skeleton
graphs [
 AnimGraph {nodes [
   AnimGraphNode_SubInput_0 = Input Pose()
   n1 = RigidBody(ComponentPose=Local To Component(LocalPose=@AnimGraphNode_SubInput_0)) {
    bClampLinearTranslationLimitToRefPose=true
    bOverrideWorldGravity=true
    overridePhysicsAsset=&/Myth_Female_Physics
    simulationSpace=BaseBoneSpace
    alphaBoolBlend {blendOption=Linear}
    baseBoneRef {boneName=b_AR_Root}
    componentAppliedLinearAccClamp=<900 1100 900>
    componentLinearAccScale=<0.35 0.4 0.3>
    componentLinearVelScale=<0.35 0.4 0.3>
    overrideWorldGravity {z=-600}
   }
   n2 = "Transform (Modify) Bone - Bone: ribbon2_01"(ComponentPose=@n1, Alpha=1, Translation="0,0,0", Rotation="0,0,-2", Scale="1,1,1") {rotationMode=BMM_Additive, alphaBoolBlend {blendOption=Linear}, boneToModify {boneName=ribbon2_01}, rotation {roll=-2}}
   n3 = "Transform (Modify) Bone - Bone: ribbon2_02"(ComponentPose=@n2, Alpha=1, Translation="0,0,0", Rotation="4,20,-8", Scale="1,1,1") {rotationMode=BMM_Additive, alphaBoolBlend {blendOption=Linear}, boneToModify {boneName=ribbon2_02}, rotation=<rot 4 20 -8>}
   n4 = "Transform (Modify) Bone - Bone: ribbon2_03"(ComponentPose=@n3, Alpha=1, Translation="0,0,0", Rotation="-8,10,-2", Scale="1,1,1") {rotationMode=BMM_Additive, alphaBoolBlend {blendOption=Linear}, boneToModify {boneName=ribbon2_03}, rotation=<rot -8 10 -2>}
   n5 = "Transform (Modify) Bone - Bone: ribbon2_04"(ComponentPose=@n4, Alpha=1, Translation="0,0,0", Rotation="-16,0,-2", Scale="1,1,1") {rotationMode=BMM_Additive, alphaBoolBlend {blendOption=Linear}, boneToModify {boneName=ribbon2_04}, rotation {pitch=-16, roll=-2}}
   Output Pose(Result=Component To Local(ComponentPose=@n5)) {layerGroup=None}
  ]}
 EventGraph {nodes [
   TryGetPawnOwner()
   Event BlueprintUpdateAnimation()
  ]}
]
variables []
